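# Sound cue for the Gumshoe pickaxe ready sound in Athena, built from the pickaxe's six swing waves.
# An equally weighted Random node picks one wave, and a Modulator slightly raises pitch and varies volume.
# The cue plays at reduced volume in the melee weapon sound class, with close projectile attenuation and a pickaxe-ready property sheet.
# It primes on load and overrides concurrency to prevent new instances once the limit is reached.
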SoundCue /BRCosmetics/Sounds/PickAxes/Gumshoe/PA_GumShoe_Ready_Athena_Cue
root=Modulator_0
nodes [
 {id=Modulator_0, children [Random_2], properties {pitchMax=1.2, pitchMin=1.1, volumeMax=1.1, volumeMin=1}}
 {
  id=Random_2
  children [WavePlayer_5, WavePlayer_10, WavePlayer_11, WavePlayer_14, WavePlayer_15, WavePlayer_17]
  properties {weights [1,1,1,1,1,1]}
 }
 {id=WavePlayer_10, properties {soundWaveAssetPtr=&/PA_GumShoe_Swing_02}}
 {id=WavePlayer_11, properties {soundWaveAssetPtr=&/PA_GumShoe_Swing_03}}
 {id=WavePlayer_14, properties {soundWaveAssetPtr=&/PA_GumShoe_Swing_04}}
 {id=WavePlayer_15, properties {soundWaveAssetPtr=&/PA_GumShoe_Swing_05}}
 {id=WavePlayer_17, properties {soundWaveAssetPtr=&/PA_GumShoe_Swing_06}}
 {id=WavePlayer_5, properties {soundWaveAssetPtr=&/PA_GumShoe_Swing_01}}
]
properties {
 attenuationSettings=&Fort_Weapon_Projectile_Close_Athena_Att
 audioPropertiesSheet=/BRCosmetics/Sounds/Settings/PropertySheets/Pickaxe/APS_Cosmetic_Pickaxe_Ready
 bOverrideConcurrency=true
 bPrimeOnLoad=true
 maxDistance=4950
 soundClassObject=&Fort_WeaponMelee
 volumeMultiplier=0.2
 concurrencyOverrides {resolutionRule=PreventNew}
}
